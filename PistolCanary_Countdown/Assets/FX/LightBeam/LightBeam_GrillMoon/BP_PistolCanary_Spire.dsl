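; Spire light beam actor for the PistolCanary Countdown event, revealed by a calendar event and removed when players teleport into the air.
; If calendar following is enabled, the authority waits for match readiness and calendar sync, then watches the configured event and sets a replicated TurnOnClient flag when active.
; The flag's OnRep toggles actor visibility where cosmetics can run.
; The authority also listens to the stateful PortalGlow event router, destroying the actor on a false payload.
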(event EventBeginPlay
  (Networking|SwitchHasAuthority
    (:Authority
      (if (Variables|Calendar|GetFollowCalendarEvents)
        (Class|FortAsyncActionAuthorityMatchReady|MatchReadyAsync
          (:OnReady
            (MCPContext|InitialCalendarSyncCompleted
              (:OnSyncComplete
                (MCPContext|WatchCalendarEvent (Variables|Calendar|GetCalendarEventToWatch)
                  (:then
                    (bind (_asynctaskproxy _payload) (GameplayEventRouter|StartListeningToStatefulEvent (GameplayEventRouter|GetGlobalEventRouter) "/PistolCanary_Countdown/Gameplay/EventRouter/PistolCanary_PortalGlow_EventRouter.PistolCanary_PortalGlow_EventRouter")
                      (:OnEventReceived
                        _)
                      (:OnSavedEventStateExists
                        _))
                    (if (not (Utilities|Struct|BreakPistolCanaryPortalGlowEventRouter _payload))
                      (Actor|DestroyActor) ; turn off when players\nteleport up into the air
                    ))
                  (:OnEventActive
                    (CallFunction|TurnonLight)))))))))))

(event Custom|TurnOnLight
  (Development|PrintString "Turn on the light!")
  (Variables|Replication|SetTurnOnClient true))

(fn OnRep_TurnOnClient
  (Networking|CanExecuteCosmeticEvents
    (:True
      (if (Variables|Replication|GetTurnOnClient)
        (Rendering|SetActorHiddenInGame false)
        (Development|PrintString "Client Side: Light Active")
        (else
          (Rendering|SetActorHiddenInGame true)
          (Development|PrintString "Client Side: Light Inactive"))))))

(fn ConstructionScript)
