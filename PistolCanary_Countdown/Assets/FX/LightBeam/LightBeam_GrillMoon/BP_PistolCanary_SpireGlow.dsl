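; Spire glow actor for PistolCanary Countdown that shows or removes a light effect based on calendar and gameplay events.
; When following the calendar, the authority waits for match ready and calendar sync, then turns the light on if the watched event is active.
; Otherwise it listens to the stateful PortalGlow router event, destroying the actor on a false payload.
; TurnOnClient replicates visibility through its OnRep; custom events also turn on or destroy the actor.
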(event Custom|PC_TurnOnLights
  (CallFunction|TurnonLight))

(event EventBeginPlay
  (bind _returnvalue (GameplayEventRouter|GetGlobalEventRouter))
  (Networking|SwitchHasAuthority
    (:Authority
      (if (Variables|Calendar|GetFollowCalendarEvents)
        (Class|FortAsyncActionAuthorityMatchReady|MatchReadyAsync
          (:OnReady
            (MCPContext|InitialCalendarSyncCompleted
              (:OnSyncComplete
                (bind _biseventactive (CalendarEvents|GetCalendarEventInformation (Variables|Calendar|GetCalendarEventToWatch)))
                (if _biseventactive
                  (CallFunction|TurnonLight)
                  (else
                    (Utilities|IsValid _returnvalue
                      (:Is%20Valid
                        (bind (_asynctaskproxy _payload) (GameplayEventRouter|StartListeningToStatefulEvent _returnvalue "/PistolCanary_Countdown/Gameplay/EventRouter/PistolCanary_PortalGlow_EventRouter.PistolCanary_PortalGlow_EventRouter")
                          (:OnEventReceived
                            _)
                          (:OnSavedEventStateExists
                            _))
                        (if (not (Utilities|Struct|BreakPistolCanaryPortalGlowEventRouter _payload))
                          (Actor|DestroyActor) ; turn off when players\nteleport up into the air
                        )))))))))))))

(event Custom|PC_TurnOffLights
  (Actor|DestroyActor) ; turn off when players\nteleport up into the air
)

(event Custom|TurnOnLight
  (Development|PrintString "Turn on the light!")
  (Variables|Replication|SetTurnOnClient true))

(fn OnRep_TurnOnClient
  (Networking|CanExecuteCosmeticEvents
    (:True
      (if (Variables|Replication|GetTurnOnClient)
        (Rendering|SetActorHiddenInGame false)
        (Development|PrintString "Client Side: Light Active")
        (else
          (Rendering|SetActorHiddenInGame true)
          (Development|PrintString "Client Side: Light Inactive"))))))

(fn ConstructionScript)
